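# Translucent, two-sided unlit material in the Season 18 biome HLOD setup, drawing a noisy color fill with bright edge lines.
# Centered UV distance, raised to powers, forms two edge masks read through their R component.
# The broader mask, scaled by Opacity, drives a fade blend.
# Tiled shield caustics blend two colors, and the tighter mask feeds a multi-stop lerp from that fill to the squared LineColor.
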Material /Game/Athena/HLODSetup/SeasonSetups/S18/M_BiomeHLOD_S18
graph {
 properties {blendMode=BLEND_Translucent, pixelDepthOffsetMode=PDOM_Legacy, shadingModel=MSM_Unlit, twoSided=true, usedShadingModels=Unlit}
 nodes [
  Constant()
  Constant() {r=1}
  n1 = TextureCoordinate()
  n2 = Abs(None=Multiply(A=Subtract(A=@n1) {constB=0.5}) {constB=2})
  n3 = Saturate(None=Power(Base=@n2, Exp=ScalarParameter() {defaultValue=20, parameterName=Power}))
  n4 = ScalarParameter() {defaultValue=55, parameterName=Mult}
  n5 = MaterialFunctionCall(Float2=Saturate(None=Multiply(A=@n3, B=@n4))) {materialFunction=&BreakOutFloat2Components}
  LinearInterpolate(Alpha=Multiply(A=Add(A=@n5.R, B=@n5.R), B=ScalarParameter() {defaultValue=0.1, parameterName=Opacity}) {constB=0.5}) {constA=1, constB=0}
  n6 = Multiply(A=TextureCoordinate() {uTiling=5, vTiling=5}, B=ScalarParameter() {defaultValue=0.4, parameterName=RepeatNoise})
  n7 = LinearInterpolate(
   A=VectorParameter() {parameterName=Color, defaultValue=<rgba 0.2567 0 0.75 1>}.RGB
   B=VectorParameter() {parameterName=Color1, defaultValue=<rgba 0.0792 0.3854 0.8646 1>}.RGB
   Alpha=Multiply(
    A=TextureSample(UVs=@n6) {texture=&T_ShieldCaustics}.R
    B=ScalarParameter() {defaultValue=2, parameterName=NoiseMult}
   )
  )
  n8 = VectorParameter() {parameterName=LineColor, defaultValue=<rgba 0.5677 0.869 1 2>}
  n9 = Saturate(None=Power(Base=@n2, Exp=ScalarParameter() {defaultValue=85, parameterName=Power0}))
  n10 = ScalarParameter() {defaultValue=4, parameterName=Mult0}
  n11 = MaterialFunctionCall(Float2=Saturate(None=Multiply(A=@n9, B=@n10))) {materialFunction=&BreakOutFloat2Components}
  MaterialFunctionCall(1=@n7, 2=@n7, 3=@n7, 4=Multiply(A=@n8.RGB, B=@n8.RGB), A=Add(A=@n11.R, B=@n11.R)) {materialFunction=&Lerp_Multiple_Float3}
 ]
}
